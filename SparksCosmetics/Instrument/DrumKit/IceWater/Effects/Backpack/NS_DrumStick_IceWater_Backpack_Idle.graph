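# Idle Niagara effect for the IceWater drumstick back bling in Sparks cosmetics: cold mist with drifting snowflakes.
# IceLightSmokeBase spawns SubUV-animated sprites in a flattened sphere that grow and sink, with alpha and tint driven by Fortnite time-of-day inputs.
# SnowFlakes spawns HDR pale-blue sprites that grow, shrink and fade under light gravity and curl noise.
# Initial velocity and camera offset modules are disabled in both emitters.
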NiagaraSystem /SparksCosmetics/Instrument/DrumKit/IceWater/Effects/Backpack/NS_DrumStick_IceWater_Backpack_Idle
dependencies {
 usedDataInterfaces [&NiagaraDataInterfaceCurve, &NiagaraDataInterfaceArrayColor, &NiagaraDataInterfaceSkeletalMesh]
 usedDynamicInputs [
  &MakeFloatFromVector
  &Lerp_Vector
  &MakeCustomFloatFromBool
 ]
 usedModules [
  &self:SystemSpawnScript.SystemScriptSource.SystemScriptGraph.NiagaraNodeAssignment_0.SetVariables_#1
  &SystemState
  &Timeline
 ]
}
emitters {
 IceLightSmokeBase {
  emitterUpdateScript [
   &EmitterState(Inactive Response=::NewEnumerator1, Life Cycle Mode=::NewEnumerator1, UseLoopDelay=false)
   &SpawnBurst_Instantaneous(Spawn Count=3, Use Spawn Probability=false, UseLoopCountLimit=true)
   &SpawnRate(Spawn Probability=0.9, SpawnRate=5, Use Spawn Probability=false)
  ]
  particleSpawnScript [
   &PartitionParticles(Number Of Partitions=3)
   &InitializeParticle(
    Color=<rgba 0.6562 0.8155 1 0.25>
    Lifetime Max=3.15
    Lifetime Min=2
    Lifetime Mode=::NewEnumerator1
    Mass Mode=::NewEnumerator3
    Position Offset=<0 0 -10>
    Sprite Rotation Mode=::NewEnumerator2
    Sprite Size Mode=::NewEnumerator1
    Uniform Sprite Size Max=50
    Uniform Sprite Size Min=25
    UsePositionOffset=false
    Write Material Random=true
   )
   &ShapeLocation(MaskSpawnGroup=false, Non Uniform Scale=<1 1 0.01>, Override Seed=false, Sphere Radius=10, UseOwnerScale=true)
   &AddVelocity(
    Cone Angle=180
    Cone Axis=<0 0 1>
    Distribution Along Cone Axis=0.25
    Inner Cone Angle=180
    Module.UseConeFalloff=true
    Module.UseVelDistribution=true
    Override Seed=false
    Speed Falloff From Cone Axis=0.5
    Velocity Mode=::NewEnumerator2
    Velocity Speed=&RandomRangeFloat(Minimum=25, Maximum=50)
   ) {enabled=false}
  ]
  particleUpdateScript [
   &ParticleState()
   /Niagara/Modules/Update/SubUV/SubUVAnimation(Number Of Frames=63)
   &ScaleSpriteSize(
    Scale Sprite Size Mode=::NewEnumerator0
    Uniform Scale Factor=&FloatFromCurve(FloatCurve=DataInterface() {propertyValues {Curve {keys [
         {arriveTangent=-0.0977, interpMode=RCIM_Cubic, leaveTangent=-0.0977, tangentMode=RCTM_User, time=0, value=0}
         {arriveTangent=2.2065, interpMode=RCIM_Cubic, leaveTangent=2.2065, tangentMode=RCTM_User, time=0.3, value=0.55}
         <at 1.0489 1 RCIM_Cubic>
        ]}}})
   )
   &ScaleColor(
    Scale Alpha=&FloatFromCurve(FloatCurve=DataInterface() {propertyValues {Curve {keys [
         <at 0 0 RCIM_Cubic>
         {arriveTangent=0.257, interpMode=RCIM_Cubic, leaveTangent=0.257, tangentMode=RCTM_User, time=0.2, value=0.65}
         {arriveTangent=0.1612, interpMode=RCIM_Cubic, leaveTangent=0.1612, tangentMode=RCTM_User, time=1.0011, value=0}
        ]}}})
    ScaleA=true
    ScaleRGB=true
   ) {enabled=false}
   &ScaleColor(
    Scale Alpha=&Multiply_Float(
     A=&FloatFromCurve(FloatCurve=DataInterface() {propertyValues {Curve {keys [
          <at 0 0 RCIM_Cubic>
          {arriveTangent=0.257, interpMode=RCIM_Cubic, leaveTangent=0.257, tangentMode=RCTM_User, time=0.2, value=0.65}
          {arriveTangent=0.1612, interpMode=RCIM_Cubic, leaveTangent=0.1612, tangentMode=RCTM_User, time=1.0011, value=0}
         ]}}})
     B=&FNTimeOfDayScalar(Night Scale=1, Morning Scale=0.35, Day Scale=0.25, Evening Scale=0.25)
    )
    Scale RGB=&FN_Select_Vector_from_Time_Of_Day(Morning=<0.9 0.9 0.9>, Day=<1 1 1>, Evening=<1 1 1>, Night=<0.5 0.5 1>)
    ScaleA=true
    ScaleRGB=true
   ) {module=ScaleColor002}
   &GravityForce(Gravity=&MakeVector(
     X=0
     Y=0
     Z=&FloatFromCurve(
      FloatCurve=DataInterface() {propertyValues {Curve {keys [<at 0 0.6 RCIM_Cubic>, {tangentMode=RCTM_User, time=1, value=1}]}}}
      Scale Curve=-40
     )
    ))
   &Drag(Use Linear Drag=true, Use Rotational Drag=true)
   &SolveForcesAndVelocity(Clamp Velocity=true, Limit Acceleration=true)
   &CameraOffset(Camera Offset Amount=15) {enabled=false}
  ]
  rendererClasses [&NiagaraSpriteRendererProperties]
  renderers [{rendererClass=&NiagaraSpriteRendererProperties}]
 }
 SnowFlakes {
  emitterUpdateScript [
   &EmitterState(Inactive Response=::NewEnumerator1, Life Cycle Mode=::NewEnumerator1, Loop Duration=0.1, UseLoopDelay=false)
   &SpawnRate(Spawn Probability=0.9, SpawnRate=10, Use Spawn Probability=true)
  ]
  particleSpawnScript [
   &PartitionParticles(Number Of Partitions=3)
   &InitializeParticle(
    Color=&SelectLinearColorFromArray(
     Array Sampling Mode=::NewEnumerator0
     Direct Array Index=?
     Color Selection Array=DataInterface() {propertyValues {ColorData [<rgba 1.2917 1.9252 2 1>, <rgba 3.6132 4.2075 5 1>]}}
     Random Seed=0
    )
    Lifetime Max=3.5
    Lifetime Min=2
    Lifetime Mode=::NewEnumerator1
    Mass Mode=::NewEnumerator3
    Position Offset=<0 0 20>
    Sprite Rotation Mode=::NewEnumerator2
    Sprite Size Mode=::NewEnumerator1
    UsePositionOffset=false
    Write Material Random=true
   )
   &ShapeLocation(MaskSpawnGroup=false, Override Seed=false, Sphere Radius=10, UseOwnerScale=true)
   /SetVariables(Particles.Rotation:NiagaraFloat=&RandomRangeFloat(Minimum=-80, Maximum=80)) {isSetParameters=true}
   &AddVelocity(
    Cone Angle=32
    Cone Axis=<0 0 1>
    Distribution Along Cone Axis=0.25
    Module.UseConeFalloff=true
    Module.UseVelDistribution=true
    Override Seed=false
    Speed Falloff From Cone Axis=0.5
    Velocity Mode=::NewEnumerator2
    Velocity Speed=&RandomRangeFloat(Minimum=500, Maximum=850)
   ) {enabled=false}
  ]
  particleUpdateScript [
   &ParticleState()
   &ScaleSpriteSize(
    Scale Sprite Size Mode=::NewEnumerator0
    Uniform Scale Factor=&FloatFromCurve(
     FloatCurve=DataInterface() {propertyValues {Curve {keys [
         {arriveTangent=-0.0977, interpMode=RCIM_Cubic, leaveTangent=-0.0977, tangentMode=RCTM_User, time=0, value=0}
         <at 0.5 0.45 RCIM_Cubic>
         <at 1.0489 0 RCIM_Cubic>
        ]}}}
     Scale Curve=0.5
    )
   )
   &ScaleColor(
    Scale Alpha=&FloatFromCurve(
     FloatCurve=DataInterface() {propertyValues {Curve {keys [
         <at 0 0 RCIM_Cubic>
         {arriveTangent=0.257, interpMode=RCIM_Cubic, leaveTangent=0.257, tangentMode=RCTM_User, time=0.2, value=0.95}
         {arriveTangent=0.1612, interpMode=RCIM_Cubic, leaveTangent=0.1612, tangentMode=RCTM_User, time=1.0011, value=0}
        ]}}}
     Scale Curve=0.5
    )
    ScaleA=true
    ScaleRGB=true
   )
   &GravityForce(Gravity=<0 0 -10>)
   &CurlNoiseForce(Bias Noise=false, Pan Noise=false)
   &SolveForcesAndVelocity(Clamp Velocity=true, Limit Acceleration=true)
   &CameraOffset(Camera Offset Amount=15) {enabled=false}
  ]
  rendererClasses [&NiagaraSpriteRendererProperties]
  renderers [{rendererClass=&NiagaraSpriteRendererProperties}]
 }
}
summary {
 systemName=NS_DrumStick_IceWater_Backpack_Idle
 emitters [
  {emitterName=IceLightSmokeBase, rendererClasses [&NiagaraSpriteRendererProperties]}
  {emitterName=SnowFlakes, rendererClasses [&NiagaraSpriteRendererProperties]}
 ]
 userVariables []
}
